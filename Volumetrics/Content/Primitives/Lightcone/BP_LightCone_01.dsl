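; Volumetric light cone actor that pairs a scaled cone mesh and dynamic material with optional spot, surface and scatter lights.
; The construction script derives end radius from length, angle and start radius, then scales the cone and feeds radius, color and light shift to the material.
; The same variables drive spotlight angles, color and intensity, scatter intensity and surface-light radius.
; Toggles destroy unwanted light components; the event graph is empty.
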(event EventBeginPlay)

(event Collision|EventActorBeginOverlap
  (params (OtherActor :type Actor%20Object%20Reference)))

(event EventTick
  (params (DeltaSeconds :type Float%20%28single-precision%29)))

(fn ConstructionScript
  (bind (_r _g _b _a) (Math|Color|BreakColor (Variables|Default|GetLight%20Color%28rgb%29%20and%20Brightness%28a%29)))
  (bind (_x _y) (Math|Vector2D|BreakVector2D (Variables|Default|GetXY%20Scale)))
  (bind _light_color_rgb__and_brightness_a (Variables|Default|GetLight%20Color%28rgb%29%20and%20Brightness%28a%29))
  (bind (_r_2 _g_2 _b_2 _a_2) (Math|Color|BreakColor _light_color_rgb__and_brightness_a))
  (bind _light_multiplier (Variables|Default|GetLight%20Multiplier))
  (bind _returnvalue (* _a_2 _light_multiplier))
  (bind _cone_angle (Variables|Default|GetCone%20Angle))
  (bind _spotlight (Variables|Default|GetSpotLight))
  (Variables|Default|SetEnd%20Radius (+ (* (Variables|Default|GetCone%20Length) (Math|Trig|Tan(Degrees) (Variables|Default|GetCone%20Angle))) (Variables|Default|GetStart%20Radius)))
  (Transformation|SetWorldScale3D (Variables|Default|GetCone) (Math|Vector|MakeVector (/ (* _x (Variables|Default|GetEnd%20Radius)) 2e+02) (/ (* _y (Variables|Default|GetEnd%20Radius)) 2e+02) (/ (Variables|Default|GetCone%20Length) 1e+03)))
  (Rendering|CreateAndAssignMID (Variables|Default|GetMID) (Variables|Default|GetMaterial)
    (:Out
      (Rendering|Material|SetMaterial (Variables|Default|GetCone) 0 (Variables|Default|GetMID))
      (Rendering|Material|SetScalarParameterValue (Variables|Default|GetMID) "StartRadius" (Variables|Default|GetStart%20Radius) :!class /Script/Engine.MaterialInstanceDynamic)
      (Rendering|Material|SetVectorParameterValue (Variables|Default|GetMID) "Color" (Math|Color|MakeColor _r _g _b (* _a (Math|Float|Max(Float) (Variables|Default|GetLight%20Shift) 1))) :!class /Script/Engine.MaterialInstanceDynamic)
      (Rendering|Material|SetScalarParameterValue (Variables|Default|GetMID) "LightShift" (Variables|Default|GetLight%20Shift) :!class /Script/Engine.MaterialInstanceDynamic)
      ;[1 0.7695 0.0498] Control Spawned Lights
      (Rendering|Lighting|SetInnerConeAngle _spotlight (* _cone_angle 0.75))
      (Rendering|Lighting|SetOuterConeAngle _spotlight (* _cone_angle 1.25))
      (Rendering|Components|Light|SetLightColor _spotlight _light_color_rgb__and_brightness_a true)
      (Rendering|Components|Light|SetIntensity _spotlight (* _returnvalue 1e+01))
      (Rendering|Components|Light|SetIntensity (Variables|Default|GetPointLight-Scatter) _returnvalue)
      (Rendering|Lighting|SetAttenuationRadius (Variables|Default|GetPointLight-Surface) (Variables|Default|GetStart%20Radius))
      (Utilities|FlowControl|Sequence
        (:then_0
          (if (not (Variables|Default|GetSurface%20Point%20Light))
            (Components|DestroyComponent (Variables|Default|GetPointLight-Surface))))
        (:then_1
          (if (not (Variables|Default|GetSpot%20Light))
            (Components|DestroyComponent (Variables|Default|GetSpotLight))))
        (:then_2
          (if (not (Variables|Default|GetScatter%20Light))
            (Components|DestroyComponent (Variables|Default|GetPointLight-Scatter))))))))
