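# Material function that analytically intersects a view ray with a Z-capped cone for volumetric light cones, despite descriptions saying cylinder.
# Ray origin and direction are transformed into the cone's axes and scaled by XYZ Scale.
# A Custom HLSL node solves the quadratic, substitutes cap planes outside Z Min-Max and clamps exit to scene depth along the camera ray.
# It outputs thickness, entry/exit distances, world and local positions, no-cap variants and a cap mask.
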MaterialFunction /Volumetrics/Content/Primitives/Lightcone/ConeIntersection
graph {
 properties {
  bExposeToLibrary=true
  description=Returns an analytical intersection for a cylinder of the specified coordinates.
  libraryCategoriesText [Math]
 }
 nodes [
  Constant3Vector() {constant=<rgba 1 0 0 0>}
  n1 = FunctionInput() {
   bUsePreviewValueAsDefault=true
   description="The X axis is the axis used to apply non uniform scale along.  Defaults to the local X axis of the mesh the material is applied to."
   inputName=XYZ Scale
   sortPriority=7
   previewValue=<1 1 0 1>
  }
  n2 = FunctionInput(Preview=Multiply(A=CameraVectorWS()) {constB=-1}) {
   bUsePreviewValueAsDefault=true
   description="The vector to calculate intersection along. Defaults to negative cameravector, or a normalized vector pointing from the camera to the world."
   inputName=Ray Direction
   sortPriority=1
  }
  n3 = NamedRerouteDeclaration(None=@n2) {name=RayDir, nodeColor=<rgba 1 0 0.04 1>}
  n4 = FunctionInput(Preview=Transform(None=Constant3Vector() {constant=<rgba 0 0 1 0>}) {transformSourceType=TRANSFORMSOURCE_Local}) {
   bUsePreviewValueAsDefault=true
   description=The Z axis is the axis along the length of the cylinder. Defaults to the local Z axis of the mesh the material is applied to.
   inputName=Cone Z Axis
   sortPriority=2
  }
  n5 = MaterialFunctionCall!(Vector1=Normalize(VectorInput=@n4), Vector2=Constant3Vector() {constant=<rgba 0.999 0.0001 0 0>}) {materialFunction=/Engine/Functions/Engine_MaterialFunctions02/Math/CreateThirdOrthogonalVector}
  n6 = FunctionInput(Preview=ShaderStageSwitch(PixelShader=VertexInterpolator(VS=@n5.Vector2).PS, VertexShader=@n5.Vector2)) {
   bUsePreviewValueAsDefault=true
   description=The Z axis is the axis along the length of the cylinder. Defaults to the local Z axis of the mesh the material is applied to.
   inputName=Cone X Axis
   sortPriority=3
  }
  n7 = FunctionOutput(None=@n6) {outputName=Cone X Axis, sortPriority=20}
  n8 = FunctionInput(Preview=ShaderStageSwitch(PixelShader=VertexInterpolator(VS=@n5.Vector3).PS, VertexShader=@n5.Vector3)) {
   bUsePreviewValueAsDefault=true
   description=The Z axis is the axis along the length of the cylinder. Defaults to the local Z axis of the mesh the material is applied to.
   inputName=Cone Y Axis
   sortPriority=4
  }
  n9 = FunctionOutput(None=@n8) {outputName=Cone Y Axis, sortPriority=21}
  n10 = FunctionOutput(None=NamedRerouteDeclaration(None=MaterialFunctionCall!(VectorToTransform=@n3, BasisX=@n7, BasisY=@n9, BasisZ=@n5.Vector1) {desc=Directional Vec, materialFunction=/Engine/Functions/Engine_MaterialFunctions02/Math/InverseTransformMatrix}.Transformed Vector 3) {name=LocalRayDir, nodeColor=<rgba 1 0.0469 0 1>}) {outputName=Local Ray Direction, sortPriority=8}
  MaterialExpressionMultiply_240 = Multiply(A=@n10, B=@n1) {desc=D}
  n11 = FunctionInput(Preview=CameraPositionWS()) {bUsePreviewValueAsDefault=true, description=Where to calculate the ray from. Defaults to Camera Position., inputName=Ray Origin}
  n12 = FunctionInput(Preview=ActorPositionWS()) {bUsePreviewValueAsDefault=true, description=The center of the cylinder. Defaults to Actor Position., inputName=Cone Origin, sortPriority=1}
  n13 = NamedRerouteDeclaration(None=Subtract(A=@n11, B=@n12)) {name=RayOrigin, nodeColor=<rgba 1 0 0.0078 1>}
  n14 = FunctionOutput(None=NamedRerouteDeclaration(None=MaterialFunctionCall(VectorToTransform=@n13, BasisX=@n7, BasisY=@n9, BasisZ=@n5.Vector1) {desc=Origin Vec, materialFunction=&InverseTransformMatrix}.Transformed Vector 3) {name=LocalRayOrigin, nodeColor=<rgba 1 0.1875 0 1>}) {outputName=Local Ray Origin, sortPriority=7}
  n15 = Constant3Vector() {constant=<rgba 0 0 1 0>}
  n16 = FunctionInput() {
   bUsePreviewValueAsDefault=true
   description="The X axis is the axis used to apply non uniform scale along.  Defaults to the local X axis of the mesh the material is applied to."
   inputName=Z Min-Max
   inputType=FunctionInput_Vector2
   sortPriority=6
   previewValue=<0 100 0 1>
  }
  n17 = MaterialFunctionCall(Float2=@n16) {materialFunction=&BreakOutFloat2Components}
  MaterialExpressionMultiply_241 = Multiply(A=@n1, B=Add(A=@n14, B=Multiply(A=@n15, B=@n17.R))) {desc=E}
  MaterialExpressionMultiply_50 = Multiply(A=@n17.R)
  MaterialExpressionMultiply_51 = Multiply(A=@n17.G)
  n18 = CameraVectorWS()
  n19 = MaterialFunctionCall() {materialFunction=&CameraDirectionVector}
  MaterialExpressionDivide_9 = Divide(A=SceneDepth().R, B=Abs(None=DotProduct(A=@n18, B=@n19.Result)))
  n20 = Custom!(
   D=@MaterialExpressionMultiply_240
   O=@MaterialExpressionMultiply_241
   Min=@MaterialExpressionMultiply_50
   Max=@MaterialExpressionMultiply_51
   SceneDepth=@MaterialExpressionDivide_9
  ) {
   code="float a =  dot(D.xy, D.xy) - D.z * D.z;\r\nfloat b = (dot(O.xy, D.xy) - O.z * D.z) * 2;\r\nfloat c =  dot(O.xy, O.xy) - O.z * O.z;\r\n\r\nfloat d = (b*b) - (4*a*c);\r\nfloat t   = sqrt( d );\r\n\r\nfloat t0 = -b - t;\r\nfloat t1 = -b + t;\r\n\r\nCapMask = 0;\r\n\r\nt0 /= 2 * a;\r\nt1 /= 2 * a;\r\n\r\nfloat dist_to_top = max( 0, (Max - O.z) / D.z );\r\nfloat dist_to_bot = max( 0, (Min - O.z) / D.z );\r\n\r\nfloat furthestcap   = dist_to_top > dist_to_bot? Max: Min;\r\nfloat furthestcapXY = length( O.xy + max(dist_to_top, dist_to_bot) * D.xy );\r\n\r\n\r\nfloat entrytest = O.z + t0 * D.z;\r\nfloat exittest =  O.z + t1 * D.z;\r\n\r\nt0 = max( 0, t0);\r\nfloat t0_nocap = t0;\r\nt1_NoCaps = t1;\r\n\r\nif( entrytest < Min || entrytest > Max )\r\n{\r\n\tt0 = min(dist_to_top, dist_to_bot);\r\n    CapMask = 1;\r\n}\r\n\r\nif( exittest  < Min || exittest  > Max )\r\n{\r\n\tt1 = max(dist_to_top, dist_to_bot);\r\n\t\r\n\tif(furthestcapXY > furthestcap ) t1 = 0;\r\n}\r\n\r\nt1 = min( t1, SceneDepth );\r\n\r\nif( d < 0) t1 = 0;\r\nfloat thickness = (t1 - t0);\r\n\r\nreturn float4(thickness, t0, t1, t0_nocap);\r\n\r\n\r\n"
   description=Cone Intersection
   outputType=CMOT_Float4
   additionalOutputs [{outputName=CapMask, outputType=CMOT_Float1}, {outputName=t1_NoCaps, outputType=CMOT_Float1}]
   inputs [
    {inputName=D, input {expression=&self:MaterialExpressionMultiply_240}}
    {inputName=O, input {expression=&self:MaterialExpressionMultiply_241}}
    {inputName=Min, input {expression=&self:MaterialExpressionMultiply_50}}
    {inputName=Max, input {expression=&self:MaterialExpressionMultiply_51}}
    {inputName=SceneDepth, input {expression=&self:MaterialExpressionDivide_9}}
   ]
  }
  FunctionOutput(None=@n20.CapMask) {outputName=CapMask, sortPriority=50}
  n21 = MaterialFunctionCall(Float3=@n20.return) {materialFunction=&BreakOutFloat3Components}
  n22 = FunctionOutput(None=@n21.B) {outputName=T1, sortPriority=4}
  n23 = NamedRerouteUsage()
  n24 = NamedRerouteUsage()
  FunctionOutput(None=Add(A=Multiply(A=@n22, B=@n23), B=@n24)) {outputName=Ray Exit World Position, sortPriority=1}
  n25 = FunctionOutput(None=@n21.G) {outputName=T0, sortPriority=3}
  FunctionOutput(None=Add(A=Multiply(A=@n25, B=@n23), B=@n24)) {outputName=Ray Entry World Position}
  FunctionOutput(None=@n21.R) {outputName=Cone Thickness, sortPriority=2}
  n26 = Multiply(A=Multiply(A=@n14) {desc=Local E})
  n27 = Multiply(A=Multiply(A=@n10) {desc=Local D})
  n28 = ComponentMask(None=@n20.return) {a=true}
  FunctionOutput(None=Add(A=@n26, B=Multiply(A=@n27, B=@n28))) {outputName=Ray Entry Local Position - NoCaps, sortPriority=5}
  n29 = NamedRerouteUsage()
  FunctionOutput(None=Add(A=NamedRerouteUsage(), B=Multiply(A=@n29, B=@n25))) {outputName=Ray Entry Local Position, sortPriority=6}
  n30 = NamedRerouteUsage()
  FunctionOutput(None=Add(A=Multiply(A=@n27, B=@n30), B=@n26)) {outputName=Ray Exit Local Position - NoCaps, sortPriority=5}
  NamedRerouteDeclaration(None=@n20.t1_NoCaps) {name=t1_NoCaps, nodeColor=<rgba 0 1 0.8359 1>}
 ]
}
