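# Translucent unlit material that renders a volumetric light cone on a scaled primitive, with depth test disabled.
# Cone dimensions come from object scale and StartRadius, feeding a ray-cone intersection function that yields entry and exit distances.
# An analytic inverse-square integral gives glow, with edge masks and Schlick phase tinting also computed.
# Switches select plain glow, four-tap panning volume noise, or a dithered raymarch with optional shadow-mask texture.
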Material /Volumetrics/Content/Primitives/Lightcone/M_LightCone_01
graph {
 properties {
  bDisableDepthTest=true
  bEnableResponsiveAA=true
  blendMode=BLEND_Translucent
  pixelDepthOffsetMode=PDOM_Legacy
  shadingModel=MSM_Unlit
  usedShadingModels=Unlit
 }
 nodes [
  n1 = NamedRerouteUsage()
  n2 = NamedRerouteUsage()
  n3 = LinearInterpolate(A=Multiply(A=NamedRerouteUsage(), B=Subtract(A=@n1, B=@n2)), Alpha=VertexColor().R) {constB=0}
  n4 = ScalarParameter() {defaultValue=8, group=Advanced, parameterName=Overscan for Near Clip}
  n5 = TextureCoordinate() {coordinateIndex=1}
  n6 = MaterialFunctionCall(Float2=Multiply(A=Add(A=@n3, B=@n4), B=ConstantBiasScale(None=@n5) {bias=-0.5, desc=-1 tp 1 Circular UVs, scale=2})) {materialFunction=&BreakOutFloat2Components}
  Add(A=Multiply(A=@n6.R, B=NamedRerouteUsage()), B=Multiply(A=@n6.G, B=NamedRerouteUsage()))
  Constant2Vector() {g=1, r=1}
  Constant2Vector() {g=5, r=5}
  n7 = Normalize(VectorInput=NamedRerouteUsage())
  n8 = Multiply(A=CameraVectorWS()) {constB=-1}
  n9 = VectorParameter() {parameterName=Color, defaultValue=<rgba 0.6 0.75 1 5000>}
  Multiply(
   A=MaterialFunctionCall!(Cosine=DotProduct(A=@n7, B=@n8), Anisotropy=ScalarParameter() {defaultValue=0.8, parameterName=Anisotropy}) {materialFunction=/Engine/Functions/Engine_MaterialFunctions01/Volumetrics/SchlickPhase}.Result
   B=Multiply(A=@n9.RGB, B=@n9.RGB)
  )
  n10 = VectorParameter() {parameterName=Pan Direction and Speed, defaultValue=<rgba 0.25 0 -1 0>}
  n11 = MaterialFunctionCall(Speed=@n10.A) {materialFunction=&TimeWithSpeedVariable}
  NamedRerouteDeclaration(None=VertexInterpolator(VS=Multiply(A=@n10.RGB, B=@n11.Time)).PS) {name=PanVector, nodeColor=<rgba 0 0.9062 1 1>}
  n12 = Normalize(VectorInput=Transform(None=Constant3Vector() {constant=<rgba 0 0 1 0>}) {transformSourceType=TRANSFORMSOURCE_Local})
  NamedRerouteDeclaration(None=ShaderStageSwitch(PixelShader=VertexInterpolator(VS=@n12).PS, VertexShader=@n12)) {name=Cone Z Axis, nodeColor=<rgba 0.0638 0.4605 1 1>}
  n13 = Divide(B=NamedRerouteUsage())
  n14 = NamedRerouteUsage()
  n15 = MaterialFunctionCall!(
   Ray Origin
   Ray Direction
   Cone Origin=NamedRerouteUsage()
   Cone Z Axis=NamedRerouteUsage()
   Cone X Axis=NamedRerouteUsage()
   Cone Y Axis=NamedRerouteUsage()
   Z Min-Max=AppendVector(A=NamedRerouteUsage(), B=NamedRerouteUsage())
   XYZ Scale=AppendVector(A=Divide(A=@n13, B=@n14), B=Constant() {r=1})
  ) {materialFunction=&/ConeIntersection}
  NamedRerouteDeclaration(None=@n15.T1) {name=T1, nodeColor=<rgba 0 0 0 1>}
  n16 = Divide(A=ComponentMask(None=NamedRerouteUsage()) {g=true, r=true}, B=NamedRerouteUsage())
  n17 = ComponentMask(None=NamedRerouteUsage()) {g=true, r=true}
  n18 = Abs(None=DotProduct(A=Normalize(VectorInput=@n16), B=Normalize(VectorInput=@n17)))
  n19 = ScalarParameter() {parameterName=EdgeFalloffPower}
  NamedRerouteDeclaration(None=LinearInterpolate(A=Power(Base=@n18, Exp=@n19), Alpha=NamedRerouteUsage()) {constAlpha=1}) {name=Edge Falloff, nodeColor=<rgba 0 0 0 1>}
  n20 = NamedRerouteUsage()
  n21 = Abs(None=Add(A=ComponentMask(None=@n20) {b=true}, B=NamedRerouteUsage()))
  n22 = Divide(A=Divide(A=ComponentMask(None=@n20) {g=true, r=true}, B=NamedRerouteUsage()), B=NamedRerouteUsage())
  n23 = MaterialFunctionCall(Vector 3=@n22, Vector 2=@n22) {materialFunction=/Engine/Functions/Engine_MaterialFunctions02/Utility/VectorLength}
  n24 = Divide(A=ScalarParameter() {defaultValue=4, group=Advanced, parameterName=Fresnel Fade Distance}, B=NamedRerouteUsage())
  n25 = Divide(A=Add(A=Subtract(A=@n21, B=@n23.V2 Length), B=@n24), B=@n24)
  n26 = Max(A=ComponentMask(None=NamedRerouteUsage()) {b=true}) {constB=0}
  n27 = ScalarParameter() {defaultValue=4, group=Advanced, parameterName=Underside Fresnel Mask}
  NamedRerouteDeclaration(None=LinearInterpolate(A=Saturate(None=@n25), Alpha=Power(Base=@n26, Exp=@n27) {desc=Prevent Fresnel Falloff from adding artifacts to bottom view.})) {name=Internal Mask, nodeColor=<rgba 0 0.6719 1 1>}
  n28 = Add(A=NamedRerouteUsage(), B=NamedRerouteUsage())
  n29 = ScalarParameter() {defaultValue=256, parameterName=Noise Scale}
  NamedRerouteDeclaration(None=Add(A=Divide(A=@n28, B=@n29), B=NamedRerouteUsage())) {name=Noise Position, nodeColor=<rgba 0 0 0 1>}
  MaterialExpressionMaterialFunctionCall_8 = MaterialFunctionCall() {materialFunction=&ObjectScale}
  MaterialExpressionScalarParameter_14 = ScalarParameter() {defaultValue=25, parameterName=StartRadius}
  n30 = Custom!(Scale=@MaterialExpressionMaterialFunctionCall_8.Scale XYZ, StartRadius=@MaterialExpressionScalarParameter_14) {
   code="float MaxRadius = max(Scale.x, Scale.y)*200;\r\nfloat2 Aspect = Scale.x > Scale.y ? float2(1.0, Scale.y / Scale.x) : float2(Scale.x / Scale.y, 1.0);\r\nfloat ConeLength = Scale.z*1000;\r\n\r\n\r\nfloat ConeTangent = (MaxRadius - StartRadius) / ConeLength;\r\nfloat ConeShift = StartRadius / ConeTangent;\r\n\r\nConeSetup = float3(ConeTangent, ConeShift, ConeLength + ConeShift);\r\n\r\nreturn float4(MaxRadius, ConeLength, Aspect);\r\n\r\n"
   description=Cone Scale and Aspect
   outputType=CMOT_Float4
   additionalOutputs [{outputName=ConeSetup, outputType=CMOT_Float3}]
   inputs [
    {inputName=Scale, input {expression=&self:MaterialExpressionMaterialFunctionCall_8}}
    {inputName=StartRadius, input {expression=&self:MaterialExpressionScalarParameter_14}}
   ]
  }
  n31 = ShaderStageSwitch(PixelShader=VertexInterpolator(VS=@n30.return).PS, VertexShader=@n30.return)
  NamedRerouteDeclaration(None=ComponentMask(None=@n31) {r=true}) {name=End Radius, nodeColor=<rgba 0 0.3191 1 1>}
  NamedRerouteDeclaration(None=@n15.CapMask) {name=CapMask, nodeColor=<rgba 0 0 0 1>}
  NamedRerouteDeclaration(None=ComponentMask(None=@n31) {g=true}) {name=Cone Length, nodeColor=<rgba 0 0.3191 1 1>}
  NamedRerouteDeclaration(None=ComponentMask(None=@n31) {a=true, b=true}) {name=Cone Aspect, nodeColor=<rgba 0 0.3191 1 1>}
  n32 = Multiply(
   A=MaterialFunctionCall(Alpha Threshold=ScalarParameter() {defaultValue=0.5, parameterName=Dither Threshold}) {materialFunction=&DitherTemporalAA}.Result
   B=ScalarParameter() {defaultValue=1, parameterName=Dither Amount}
  )
  NamedRerouteDeclaration(None=Divide(A=@n32) {constB=4}) {name=Dither, nodeColor=<rgba 1 0.6094 0 1>}
  NamedRerouteDeclaration(None=@n15.Ray Entry World Position) {name=Cone Ray Entry, nodeColor=<rgba 0 0 0 1>}
  n33 = ShaderStageSwitch(PixelShader=VertexInterpolator(VS=@n30.ConeSetup).PS, VertexShader=@n30.ConeSetup)
  NamedRerouteDeclaration(None=ComponentMask(None=@n33) {r=true}) {name=Cone Tangent, nodeColor=<rgba 0.0078 1 0 1>}
  MaterialExpressionReroute_10 = Reroute(None=TextureObjectParameter() {parameterName=NoiseVolumeTexture, texture=/Engine/EngineSky/VolumetricClouds/T_VolumeNoiseErosion32})
  MaterialExpressionTextureObjectParameter_1 = TextureObjectParameter() {parameterName=ShadowMaskTexture, texture=/Engine/ArtTools/RenderToTexture/Textures/T_EV_BlankWhite_01}
  MaterialExpressionScalarParameter_35 = ScalarParameter() {defaultValue=32, parameterName=Steps}
  MaterialExpressionReroute_5 = Reroute(None=AppendVector(A=NamedRerouteUsage(), B=NamedRerouteUsage()))
  MaterialExpressionNamedRerouteUsage_63 = NamedRerouteUsage()
  MaterialExpressionMultiply_41 = Multiply(A=CameraVectorWS()) {constB=-1}
  MaterialExpressionReroute_6 = Reroute(None=ScalarParameter() {defaultValue=256, parameterName=Noise Scale})
  MaterialExpressionNamedRerouteUsage_74 = NamedRerouteUsage()
  MaterialExpressionNamedRerouteUsage_65 = NamedRerouteUsage()
  MaterialExpressionAppendVector_6 = AppendVector(A=NamedRerouteUsage(), B=NamedRerouteUsage())
  MaterialExpressionAppendVector_8 = AppendVector(A=NamedRerouteUsage(), B=NamedRerouteUsage())
  MaterialExpressionMultiply_43 = Multiply(A=NamedRerouteUsage(), B=ScalarParameter() {defaultValue=0.5, parameterName=LightShift})
  MaterialExpressionNamedRerouteUsage_61 = NamedRerouteUsage()
  MaterialExpressionNamedRerouteUsage_60 = NamedRerouteUsage()
  MaterialExpressionAppendVector_9 = AppendVector(A=ScalarParameter() {defaultValue=1, parameterName=Noise Contrast}, B=ScalarParameter() {defaultValue=2, parameterName=ShadowMaskContrast})
  n34 = ScalarParameter() {defaultValue=0.5, parameterName=Shadow Mask Pan Speed}
  MaterialExpressionAppendVector_17 = AppendVector(
   A=Add(
    A=MaterialFunctionCall(Speed=@n34) {materialFunction=&TimeWithSpeedVariable}.Time
    B=ScalarParameter() {parameterName=Shadowmask UV Offset}
   )
   B=ScalarParameter() {defaultValue=1, parameterName=ShadowMask UV Scale}
  )
  n35 = NamedRerouteUsage()
  n36 = ScalarParameter() {defaultValue=256, parameterName=Noise Scale}
  MaterialExpressionSubtract_7 = Subtract(A=Add(A=NamedRerouteUsage(), B=Divide(A=@n35, B=@n36)), B=NamedRerouteUsage())
  n37 = Custom!(
   Tex=@MaterialExpressionReroute_10
   ShadowMask=@MaterialExpressionTextureObjectParameter_1
   Steps=@MaterialExpressionScalarParameter_35
   Intersection=@MaterialExpressionReroute_5
   RO=@MaterialExpressionNamedRerouteUsage_63
   RD=@MaterialExpressionMultiply_41
   NoiseScale=@MaterialExpressionReroute_6
   Dither=@MaterialExpressionNamedRerouteUsage_74
   EnableRayMarch=@MaterialExpressionNamedRerouteUsage_65
   ZExtents=@MaterialExpressionAppendVector_6
   Radii=@MaterialExpressionAppendVector_8
   LightShift=@MaterialExpressionMultiply_43
   RO_Local=@MaterialExpressionNamedRerouteUsage_61
   RD_Local=@MaterialExpressionNamedRerouteUsage_60
   Contrasts=@MaterialExpressionAppendVector_9
   ShadowMaskUVBiasScale=@MaterialExpressionAppendVector_17
   Panner=@MaterialExpressionSubtract_7
  ) {
   code="float Accum = 0;\r\n\r\nfloat TotalDistsance = max(0, Intersection.y - Intersection.x);\r\n\r\nfloat StepSize = (TotalDistsance / Steps);\r\nfloat ConeLength = (ZExtents.y - ZExtents.x);\r\nRO += float3(0,0,LightShift);\r\n\r\nfor (int i = 0; i < Steps; i++)\r\n{\r\n\tfloat3 CurUVW = RO + (RD * (i+Dither)* StepSize);\r\n\t\r\n\tfloat CurSample = Tex.SampleLevel(TexSampler, (CurUVW / NoiseScale) + Panner, 0).r;\r\n\tCurSample = pow(CurSample, Contrasts.x);\r\n\t\r\n\tfloat InvSq = 1 / (length(CurUVW) * length(CurUVW));\r\n\t\r\n\t##if RAYMARCHSHADOWMASK\r\n\t\r\n\t\tfloat3 CurUVW_Local = RO_Local + (RD_Local * (i+Dither)* StepSize);\r\n\t\tfloat normalizedZ = (CurUVW_Local.z) / ConeLength;\r\n\t\tfloat CurRadius = lerp(Radii.x, Radii.y, normalizedZ);\r\n\t\tfloat2 MaskUV = ((CurUVW_Local.xy / CurRadius) / 2) + 0.5;\r\n\t\tMaskUV = (MaskUV + ShadowMaskUVBiasScale.x) * ShadowMaskUVBiasScale.y;\r\n\t\t\r\n\t\tfloat CurShadow = ShadowMask.SampleLevel(ShadowMaskSampler, MaskUV, 0);\r\n\t\tCurSample *= pow(CurShadow, Contrasts.y);\r\n\t\r\n\t##endif\r\n\r\n\tAccum += (CurSample * InvSq) * StepSize;\r\n} \r\n\r\n\r\nreturn Accum;\r\n\r\n//return  Tex.SampleLevel(TexSampler, RO / NoiseScale, 0).r;\r\n\t"
   inputs [
    {inputName=Tex, input {expression=&self:MaterialExpressionReroute_10}}
    {inputName=ShadowMask, input {expression=&self:MaterialExpressionTextureObjectParameter_1}}
    {inputName=Steps, input {expression=&self:MaterialExpressionScalarParameter_35}}
    {inputName=Intersection, input {expression=&self:MaterialExpressionReroute_5}}
    {inputName=RO, input {expression=&self:MaterialExpressionNamedRerouteUsage_63}}
    {inputName=RD, input {expression=&self:MaterialExpressionMultiply_41}}
    {inputName=NoiseScale, input {expression=&self:MaterialExpressionReroute_6}}
    {inputName=Dither, input {expression=&self:MaterialExpressionNamedRerouteUsage_74}}
    {inputName=EnableRayMarch, input {expression=&self:MaterialExpressionNamedRerouteUsage_65}}
    {inputName=ZExtents, input {expression=&self:MaterialExpressionAppendVector_6}}
    {inputName=Radii, input {expression=&self:MaterialExpressionAppendVector_8}}
    {inputName=LightShift, input {expression=&self:MaterialExpressionMultiply_43}}
    {inputName=RO_Local, input {expression=&self:MaterialExpressionNamedRerouteUsage_61}}
    {inputName=RD_Local, input {expression=&self:MaterialExpressionNamedRerouteUsage_60}}
    {inputName=Contrasts, input {expression=&self:MaterialExpressionAppendVector_9}}
    {
     inputName=ShadowMaskUVBiasScale
     input {expression=&self:MaterialExpressionAppendVector_17}
    }
    {inputName=Panner, input {expression=&self:MaterialExpressionSubtract_7}}
   ]
  }
  NamedRerouteDeclaration(None=Multiply(A=@n37, B=ScalarParameter() {defaultValue=1, parameterName=RayMarchMultiplier})) {name=RayMarchedCone, nodeColor=<rgba 1 0 0.6562 1>}
  NamedRerouteDeclaration(None=@n15.Ray Entry Local Position) {name=Ray Entry Local, nodeColor=<rgba 0 0 0 1>}
  NamedRerouteDeclaration(None=ComponentMask(None=@n33) {g=true}) {name=ZMin, nodeColor=<rgba 0.0078 1 0 1>}
  NamedRerouteDeclaration(None=ComponentMask(None=@n33) {b=true}) {name=ZMax, nodeColor=<rgba 0.0078 1 0 1>}
  NamedRerouteDeclaration(None=ScalarParameter() {defaultValue=25, parameterName=StartRadius}) {name=Start radius, nodeColor=<rgba 0 0.3191 1 1>}
  NamedRerouteDeclaration(None=StaticSwitchParameter(
    True=Custom!() {code="##define RAYMARCHSHADOWMASK 1\r\nreturn 0;", description="##define RAYMARCHSHADOWMASK 1", inputs []}
    False=Custom() {code="##define RAYMARCHSHADOWMASK 0\r\nreturn 0;", description="##define RAYMARCHSHADOWMASK 0", inputs []}
   ) {parameterName=RayMarchShadowMask}) {name=ShadowMask Define, nodeColor=<rgba 0 0 0 1>}
  n38 = Normalize(VectorInput=Transform(None=Constant3Vector() {constant=<rgba 1 0 0 0>}) {transformSourceType=TRANSFORMSOURCE_Local})
  NamedRerouteDeclaration(None=ShaderStageSwitch(PixelShader=VertexInterpolator(VS=@n38).PS, VertexShader=@n38)) {name=Cone X Axis, nodeColor=<rgba 0.0638 0.4605 1 1>}
  NamedRerouteDeclaration(None=@n15.Ray Exit World Position) {name=Cone Ray Exit, nodeColor=<rgba 0 0 0 1>}
  n39 = Normalize(VectorInput=Transform(None=Constant3Vector() {constant=<rgba 0 1 0 0>}) {transformSourceType=TRANSFORMSOURCE_Local})
  NamedRerouteDeclaration(None=ShaderStageSwitch(PixelShader=VertexInterpolator(VS=@n39).PS, VertexShader=@n39)) {name=Cone Y Axis, nodeColor=<rgba 0.0638 0.4605 1 1>}
  NamedRerouteDeclaration(None=@n32) {name=Dither for Raymarch, nodeColor=<rgba 0 0.8594 1 1>}
  NamedRerouteDeclaration(None=@n15.Local Ray Origin) {name=Ray Origin Local, nodeColor=<rgba 0 0 0 1>}
  NamedRerouteDeclaration(None=@n15.Local Ray Direction) {name=Ray Direction Local, nodeColor=<rgba 0 0 0 1>}
  NamedRerouteDeclaration(None=TransformPosition(None=Constant3Vector()) {transformType=TRANSFORMPOSSOURCE_World}) {name=Cone Origin, nodeColor=<rgba 0.0638 0.4605 1 1>}
  NamedRerouteDeclaration(None=@n15.Ray Entry Local Position - NoCaps) {name=Ray Entry Local-NoCaps, nodeColor=<rgba 0 0 0 1>}
  n40 = Add(A=Constant(), B=NamedRerouteUsage())
  n41 = Multiply(
   A=Divide(
    A=ScalarParameter() {defaultValue=256, parameterName=Noise Trace Distance}
    B=ScalarParameter() {defaultValue=256, parameterName=Noise Scale}
   )
   B=Multiply(A=CameraVectorWS()) {constB=-1}
  )
  n42 = TextureSampleParameterVolume!(UVs=Add(A=Multiply(A=@n41, B=@n40), B=NamedRerouteUsage()), Apply View MipBias) {parameterName=NoiseVolumeTexture, samplerType=SAMPLERTYPE_Masks, texture=&T_VolumeNoiseErosion32}
  n43 = Add(A=Constant() {r=0.25}, B=NamedRerouteUsage())
  n44 = TextureSampleParameterVolume(UVs=Add(A=Multiply(A=@n41, B=@n43), B=NamedRerouteUsage())) {parameterName=NoiseVolumeTexture, samplerType=SAMPLERTYPE_Masks, texture=&T_VolumeNoiseErosion32}
  n45 = Add(A=@n42.R, B=@n44.R)
  n46 = Add(A=Constant() {r=0.5}, B=NamedRerouteUsage())
  n47 = TextureSampleParameterVolume(UVs=Add(A=Multiply(A=@n41, B=@n46), B=NamedRerouteUsage())) {parameterName=NoiseVolumeTexture, samplerType=SAMPLERTYPE_Masks, texture=&T_VolumeNoiseErosion32}
  n48 = Add(A=Constant() {r=0.75}, B=NamedRerouteUsage())
  n49 = TextureSampleParameterVolume(UVs=Add(A=Multiply(A=@n41, B=@n48), B=NamedRerouteUsage())) {parameterName=NoiseVolumeTexture, samplerType=SAMPLERTYPE_Masks, texture=&T_VolumeNoiseErosion32}
  n50 = Divide(A=Add(A=Add(A=@n45, B=@n47.R), B=@n49.R)) {constB=4}
  NamedRerouteDeclaration(None=Power(Base=Saturate(None=@n50), Exp=ScalarParameter() {defaultValue=1, parameterName=Noise Contrast})) {name=VolumeNoise, nodeColor=<rgba 0.0335 0.5098 1 1>}
  NamedRerouteDeclaration(None=@n15.T0) {name=T0, nodeColor=<rgba 0 0 0 1>}
  ScalarParameter() {defaultValue=0.267, parameterName=ConeTangent}
  ScalarParameter() {defaultValue=186, parameterName=ZMin}
  ScalarParameter() {defaultValue=1186, parameterName=ZMax}
  ScalarParameter() {parameterName=RadiusEnd}
  n51 = NamedRerouteUsage()
  MaterialExpressionNamedRerouteUsage_15 = NamedRerouteUsage()
  MaterialExpressionNamedRerouteUsage_14 = NamedRerouteUsage()
  n52 = Constant()
  n53 = Multiply(A=NamedRerouteUsage(), B=ScalarParameter() {defaultValue=0.5, parameterName=LightShift})
  MaterialExpressionAdd_1 = Add(
   A=NamedRerouteUsage()
   B=NamedRerouteDeclaration(None=MaterialFunctionCall(X=@n52, Y=@n52, Z=@n53) {materialFunction=&MakeFloat3}.Result) {name=Light Shift, nodeColor=<rgba 1 0.6757 0 1>}
  ) {constB=0}
  MaterialExpressionNamedRerouteUsage_9 = NamedRerouteUsage()
  MaterialExpressionScalarParameter_6 = ScalarParameter() {defaultValue=0.001, desc=Tiny offset to prevent huge brightness values, group=Advanced, parameterName=TinyOffset}
  n54 = Custom!(
   t0=@MaterialExpressionNamedRerouteUsage_15
   t1=@MaterialExpressionNamedRerouteUsage_14
   RayOrigin=@MaterialExpressionAdd_1
   RayDirection=@MaterialExpressionNamedRerouteUsage_9
   f=@MaterialExpressionScalarParameter_6
  ) {
   code="float b = dot(RayOrigin,RayDirection);\r\nfloat c = dot(RayOrigin,RayOrigin);\r\n\r\nt0 = max( t0, 0.0f );\r\n\r\n//tiny offset param to avoid infinite brightness\r\nc+= f;\r\n\r\nfloat h2 = sqrt( (c) - (b*b) );\r\nfloat integral_t0 = ( atanFast( ( t0 + b ) / h2 ) ) / h2;\r\nfloat integral_t1 = ( atanFast( ( t1 + b ) / h2 ) ) / h2;\r\n\r\nreturn (integral_t1 - integral_t0);\r\n"
   description=Inverse Square Falloff
   outputType=CMOT_Float1
   inputs [
    {inputName=t0, input {expression=&self:MaterialExpressionNamedRerouteUsage_15}}
    {inputName=t1, input {expression=&self:MaterialExpressionNamedRerouteUsage_14}}
    {inputName=RayOrigin, input {expression=&self:MaterialExpressionAdd_1}}
    {inputName=RayDirection, input {expression=&self:MaterialExpressionNamedRerouteUsage_9}}
    {inputName=f, input {expression=&self:MaterialExpressionScalarParameter_6}}
   ]
  }
  n55 = Multiply(A=ComponentMask(None=@n54) {r=true}, B=NamedRerouteUsage())
  StaticSwitchParameter(
   True=StaticSwitchParameter(True=NamedRerouteUsage(), False=Multiply(A=@n51, B=@n55) {constA=1, desc=O}) {parameterName=RayMarchedNoise}
   False=@n55
  ) {defaultValue=true, parameterName=Use Noise}
  VectorParameter() {parameterName=asdf, defaultValue=<rgba 0 1 0 0>}
 ]
}
